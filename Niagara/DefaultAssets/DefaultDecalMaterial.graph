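# Default Niagara decal material: an unlit, translucent deferred decal that forms a soft circular spot.
# A sphere mask around the UV center, shaped by Radius and Hardness parameters, draws the circle.
# That mask is multiplied by the decal lifetime opacity, routed through a float2 breakout's R channel, so the spot follows the decal's fade.
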Material /Niagara/DefaultAssets/DefaultDecalMaterial
graph {
 properties {
  bEnableResponsiveAA=true
  bOutputTranslucentVelocity=true
  bUsedWithNiagaraRibbons=true
  bUsedWithNiagaraSprites=true
  bUsedWithParticleSprites=true
  blendMode=BLEND_Translucent
  materialDomain=MD_DeferredDecal
  pixelDepthOffsetMode=PDOM_Legacy
  shadingModel=MSM_Unlit
  twoSided=true
  usedShadingModels=Unlit
 }
 nodes [
  Constant()
  Multiply(
   A=SphereMask(
    A=TextureCoordinate()
    B=Constant() {r=0.5}
    Radius=ScalarParameter() {defaultValue=0.5, parameterName=Radius}
    Hardness=ScalarParameter() {defaultValue=0.5, parameterName=Hardness}
   )
   B=MaterialFunctionCall(Float2=DecalLifetimeOpacity().Opacity) {materialFunction=&BreakOutFloat2Components}.R
  )
  TextureCoordinate()
 ]
}
