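# Default additive, unlit, two-sided textured sprite material for Niagara particles, also flagged for ribbons and particle sprites.
# Particle color multiplies a Texture parameter, with RGB and alpha combined separately.
# A clamped TwoSidedSign blends the color between a dimmed copy for back faces and full strength for front faces.
# Responsive AA and translucent velocity output are enabled.
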Material /Niagara/DefaultAssets/DefaultTextureSpriteMaterial
graph {
 properties {
  bEnableResponsiveAA=true
  bOutputTranslucentVelocity=true
  bUsedWithNiagaraRibbons=true
  bUsedWithNiagaraSprites=true
  bUsedWithParticleSprites=true
  blendMode=BLEND_Additive
  pixelDepthOffsetMode=PDOM_Legacy
  shadingModel=MSM_Unlit
  twoSided=true
  usedShadingModels=Unlit
 }
 nodes [
  n1 = ParticleColor()
  n2 = TextureSampleParameter2D() {parameterName=Texture}
  n3 = Multiply(A=@n1.RGB, B=@n2.RGB)
  LinearInterpolate(A=Multiply(A=@n3) {constB=0.15}, B=@n3, Alpha=Clamp(None=TwoSidedSign()))
  Multiply(A=@n1.A, B=@n2.A)
 ]
}
