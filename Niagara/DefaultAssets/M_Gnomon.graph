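# M_Gnomon is an unlit Niagara default material for mesh particles, apparently for a gnomon mesh.
# It fakes shading by dotting the unnormalized offset from actor position with the vertex normal, clamping it with a raised floor, and multiplying by vertex color.
# The Desat parameter desaturates that result, which is lerped against particle color using particle RGB as both the A input and the per-channel alpha.
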Material /Niagara/DefaultAssets/M_Gnomon
graph {
 properties {
  bUsedWithMeshParticles=true
  bUsedWithNiagaraMeshParticles=true
  pixelDepthOffsetMode=PDOM_Legacy
  shadingModel=MSM_Unlit
  usedShadingModels=Unlit
 }
 nodes [
  n1 = ParticleColor()
  n2 = VertexColor()
  n3 = WorldPosition()
  n4 = ActorPositionWS()
  n5 = Clamp(None=DotProduct(A=Subtract(A=@n3.XYZ, B=@n4), B=VertexNormalWS())) {minDefault=0.25}
  LinearInterpolate(A=@n1.RGB, B=Desaturation(None=Multiply(A=@n2, B=@n5), Fraction=ScalarParameter() {defaultValue=0.125, parameterName=Desat}), Alpha=@n1.RGB)
 ]
}
