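# M_NiagaraUVDebug is an unlit, two-sided default Niagara material, named for debugging UV mapping on particles.
# It is enabled for sprite, ribbon and mesh particle renderers.
# The graph holds a single texture sample of T_NiagaraUVDebug, with no other nodes or UV manipulation.
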Material /Niagara/DefaultAssets/M_NiagaraUVDebug
graph {
 properties {
  bUsedWithNiagaraMeshParticles=true
  bUsedWithNiagaraRibbons=true
  bUsedWithNiagaraSprites=true
  pixelDepthOffsetMode=PDOM_Legacy
  shadingModel=MSM_Unlit
  twoSided=true
  usedShadingModels=Unlit
 }
 nodes [TextureSample() {texture=&/T_NiagaraUVDebug}]
}
